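# Translucent UI material that draws an animated glowing backing for the squad directive in the Battle Royale HUD squad list.
# A 9-slice box SDF with two rounded corners yields a fill and an eased outer glow, whose opacity pulses on a sine.
# A slightly rotated horizontal gradient fades the opacity and, multiplied by the shape, drives a three-color blend.
# The blend's blues and cyan, partly from the shared UI parameter collection, oscillate over time.
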Material /BattleRoyaleGameUI/HUD/SquadList/Assets/M_UI_SquadDirectiveBacking
graph {
 properties {bAutomaticallySetUsageInEditor=false, blendMode=BLEND_Translucent, materialDomain=MD_UI, usedShadingModels=Default Lit}
 nodes [
  ComponentMask(None=CollectionParameter() {collection=/BattleRoyaleGameUI/Shared/MPC_BRGameUI, parameterName=White}) {b=true, g=true, r=true}
  Constant3Vector() {constant=<rgba 0.0369 1 0.9131 1>}
  Constant() {r=1}
  n1 = ComponentMask(None=CollectionParameter() {collection=&MPC_BRGameUI, parameterName=Blue}) {b=true, g=true, r=true}
  n2 = MaterialFunctionCall(Speed=ScalarParameter() {defaultValue=0.5, parameterName=Speed}) {materialFunction=&DebugTimeSine}
  n3 = Constant3Vector() {constant=<rgba 0 0.3333 1 1>}
  n4 = CollectionParameter() {collection=&MPC_BRGameUI, parameterName=Cyan}
  n5 = MaterialFunctionCall!(UVs) {materialFunction=/FortUILibrary/MaterialFunctions/Basics/MF_UI_9SliceUVs}
  n6 = ScalarParameter() {defaultValue=0.5, parameterName=Size}
  n7 = MaterialFunctionCall(UVs=@n5.UVs, Size=@n6, CornerRoundness=Constant4Vector() {constant=<rgba 0.5 0.5 0 0>}.RGBA) {materialFunction=&MF_UI_Box}
  n8 = Saturate(None=MaterialFunctionCall(
    DistanceField=@n7.SDF
    Threshold=ScalarParameter() {defaultValue=0.25, parameterName=GlowThreshold}
    Smoothing=ScalarParameter() {defaultValue=0.5, parameterName=GlowSmoothing}
   ) {materialFunction=&MF_UI_AntiAliasing}.Inner)
  n9 = ScalarParameter() {defaultValue=0.3333, parameterName=GlowOpacity}
  n10 = Multiply(A=@n9) {constB=1.5}
  n11 = MaterialFunctionCall(Speed=ScalarParameter() {defaultValue=0.5, parameterName=Speed}) {materialFunction=&DebugTimeSine}
  n12 = LinearInterpolate(
   A=Multiply(
    A=MaterialFunctionCall("Value (0-1)"=@n8) {materialFunction=&MF_UI_EaseCurves}.EaseInCubic
    B=LinearInterpolate(A=@n9, B=@n10, Alpha=@n11.Result)
   )
   B=ScalarParameter() {defaultValue=1, parameterName=FillOpacity}
   Alpha=Saturate(None=@n7.Fill)
  )
  n13 = MaterialFunctionCall() {materialFunction=&GetUserInterfaceUV}
  n14 = Constant() {r=0.0125}
  n15 = OneMinus(None=ComponentMask(None=MaterialFunctionCall(UVs=@n13.Normalized UV, "Rotation Angle (0-1)"=@n14) {materialFunction=&CustomRotator}.Rotated Values) {r=true})
  MaterialFunctionCall(
   A=LinearInterpolate(A=Constant3Vector() {constant=<rgba 0 0 1 0>}, B=@n1, Alpha=@n2.Result) {constA=0.75}
   B=LinearInterpolate(A=@n1, B=@n3, Alpha=@n2.Result) {constA=0.75}
   C=LinearInterpolate(A=@n3, B=ComponentMask(None=@n4) {b=true, g=true, r=true}, Alpha=@n2.Result) {constA=0.75}
   Alpha=Multiply(A=@n12, B=@n15)
  ) {materialFunction=&3ColorBlend}
  MaterialFunctionCall(UVs=@n5.UVs, Size=@n6) {materialFunction=&MF_UI_Circle}
  n16 = Constant() {r=-0.125}
  n17 = MaterialFunctionCall("Value (0-1)"=@n15) {materialFunction=&MF_UI_EaseCurves}
  Multiply(A=@n12, B=Saturate(None=LinearInterpolate(A=@n16, Alpha=@n17.EaseInOutSine) {constB=1.25}))
 ]
}
