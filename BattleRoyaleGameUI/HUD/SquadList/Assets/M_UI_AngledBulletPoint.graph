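# UI material that draws an angled bullet point with a soft shadow, filed under the Battle Royale squad list HUD.
# Rotated, perspective-skewed and offset UVs feed a box SDF, which anti-aliasing functions turn into a crisp fill mask and a smoothed shadow mask.
# The fill mask blends shadow into fill for brightness and alpha.
# Alpha-composite blending with reduced fill alpha makes the fill partly additive.
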Material /BattleRoyaleGameUI/HUD/SquadList/Assets/M_UI_AngledBulletPoint
graph {
 properties {bAutomaticallySetUsageInEditor=false, blendMode=BLEND_AlphaComposite, materialDomain=MD_UI, usedShadingModels=Default Lit}
 nodes [
  n1 = ScalarParameter() {defaultValue=0.1, parameterName=Smoothing}
  n2 = MaterialFunctionCall(UVs=TextureCoordinate(), "Rotation Angle (0-1)"=ScalarParameter() {defaultValue=-0.085, parameterName=OverallRotation}) {materialFunction=&CustomRotator}
  n3 = ScalarParameter() {defaultValue=-0.1, parameterName=PerspectiveRotation}
  n4 = ScalarParameter() {defaultValue=-0.5, parameterName=PerspectiveAmount}
  n5 = ScalarParameter() {defaultValue=-0.075, parameterName=OffsetX}
  n6 = ScalarParameter() {defaultValue=-0.08, parameterName=OffsetY}
  n7 = MaterialFunctionCall(
   UVs=Add(
    A=MaterialFunctionCall!(UV=@n2.Rotated Values, Perspective Rotation=@n3, Perspective Amount=@n4) {materialFunction=/FortUILibrary/MaterialFunctions/Basics/MF_UI_Perspective}.Perspective UVs
    B=AppendVector(A=@n5, B=@n6)
   )
   Size=ScalarParameter() {defaultValue=0.425, parameterName=Size}
  ) {materialFunction=&MF_UI_Box}
  n8 = Saturate(None=MaterialFunctionCall(DistanceField=@n7.SDF, Threshold=@n1, Smoothing=@n1) {materialFunction=&MF_UI_AntiAliasing}.Inner)
  n9 = NamedRerouteDeclaration(None=Saturate(None=MaterialFunctionCall(DistanceField=@n7.SDF) {materialFunction=&MF_UI_AntiAliasing}.Inner)) {name=FillMask, nodeColor=<rgba 0 0 0 1>}
  LinearInterpolate(
   A=Multiply(
    A=NamedRerouteDeclaration(None=@n8) {name=ShadowMask, nodeColor=<rgba 0 0 0 1>}
    B=ScalarParameter() {defaultValue=0.1, parameterName=ShadowOpacity}
   )
   B=OneMinus(None=ScalarParameter() {defaultValue=0.125, parameterName=Fill_AdditiveMixRatio})
   Alpha=@n9
  )
  LinearInterpolate(B=ScalarParameter() {defaultValue=0.875, parameterName=Fill_DIm}, Alpha=@n9)
  ScalarParameter() {defaultValue=0.25, parameterName=Threshold}
 ]
}
