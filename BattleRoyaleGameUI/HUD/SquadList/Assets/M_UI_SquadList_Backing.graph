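# Translucent UI material for the squad list backing in the Battle Royale HUD, drawing a rounded box with a thin side bar.
# Shared adaptive colors and a collection red set backing and bar color, with IsRed and IsDead parameters blending state variants.
# A rotated, aspect-corrected gradient with quint easing fades backing opacity, while an anti-aliased, slightly rotated threshold masks the bar.
# Low-end switchers substitute constants for several opacity parameters.
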Material /BattleRoyaleGameUI/HUD/SquadList/Assets/M_UI_SquadList_Backing
graph {
 properties {bAutomaticallySetUsageInEditor=false, blendMode=BLEND_Translucent, materialDomain=MD_UI, usedShadingModels=Default Lit}
 nodes [
  n1 = MaterialFunctionCall!(BufferSource) {materialFunction=/BattleRoyaleGameUI/Shared/MF_UI_AdaptiveColors}
  n2 = NamedRerouteDeclaration(None=@n1.Dark) {name=ProcessDark, nodeColor=<rgba 1 0.9609 0 1>}
  n3 = LinearInterpolate(A=@n2, B=LinearInterpolate(A=@n2, B=NamedRerouteUsage()) {constAlpha=0.125}, Alpha=NamedRerouteUsage())
  n4 = ScalarParameter() {parameterName=IsRed}
  n5 = MaterialFunctionCall!(Default=@n1.Dim, LowEnd=@n1.Dim) {materialFunction=/BattleRoyaleGameUI/Shared/MF_UI_LowEnd_ValueSwitcher}
  LinearInterpolate(
   A=LinearInterpolate(A=LinearInterpolate(A=@n2, B=@n3, Alpha=@n4), B=NamedRerouteUsage(), Alpha=ScalarParameter() {parameterName=IsDead})
   B=LinearInterpolate(
    A=NamedRerouteDeclaration(None=@n5.Result) {name=ProcessDim, nodeColor=<rgba 0.6406 0 1 1>}
    B=NamedRerouteUsage()
    Alpha=ScalarParameter() {parameterName=IsRed}
   )
   Alpha=NamedRerouteUsage()
  )
  n6 = MaterialFunctionCall(Default=ScalarParameter() {defaultValue=0.75, parameterName=Opacity_Start}, LowEnd=Constant() {r=0.464}) {materialFunction=&MF_UI_LowEnd_ValueSwitcher}
  n7 = MaterialFunctionCall(Default=ScalarParameter() {defaultValue=0.25, parameterName=BackingDeadOpacity}, LowEnd=Constant() {r=0.121}) {materialFunction=&MF_UI_LowEnd_ValueSwitcher}
  n8 = ScalarParameter() {parameterName=IsDead}
  n9 = MaterialFunctionCall() {materialFunction=&GetUserInterfaceUV}
  n10 = MaterialFunctionCall(UV=@n9.Normalized UV) {materialFunction=&MF_UI_UVScaling}
  n11 = ScalarParameter() {defaultValue=-0.23, parameterName=FadeRotation}
  MaterialExpressionMaterialFunctionCall_29 = MaterialFunctionCall!(Dimensions=@n9.Pixel Size) {materialFunction=/FortUILibrary/MaterialFunctions/Basics/MF_UI_GetAspectRatio}
  n12 = Convert(None=@MaterialExpressionMaterialFunctionCall_29.Max Ratio) {
   nodeName=Break Float2
   convertInputs [{
     type=Vector2
     expressionInput {
      expression=&self:MaterialExpressionMaterialFunctionCall_29
      inputName=None
      mask=0
      maskA=0
      maskB=0
      maskG=0
      maskR=0
      outputIndex=2
     }
    }]
   convertOutputs [{type=Scalar}, {type=Scalar}]
  }
  n13 = Multiply(A=@n12, B=Abs(None=Sine(None=@n11)))
  n14 = Multiply(A=@n12, B=Abs(None=Cosine(None=@n11)))
  n15 = ComponentMask(None=MaterialFunctionCall(
    UVs=MaterialFunctionCall(UVs=@n10.Scaled to Fit, "Rotation Angle (0-1)"=@n11) {materialFunction=&CustomRotator}.Rotated Values
    Texture Scale=Add(A=@n13, B=@n14)
   ) {materialFunction=&ScaleUVsByCenter}.UVs) {g=true}
  n16 = ScalarParameter() {defaultValue=0.05, parameterName=AdjustOpacityCurve_Stop}
  n17 = ScalarParameter() {defaultValue=0.9, parameterName=AdjustOpacityCurve_Start}
  n18 = NamedRerouteDeclaration(None=Saturate(None=MaterialFunctionCall!(Input=@n15, Input Low=@n16, Input High=@n17) {materialFunction=/Engine/Functions/Engine_MaterialFunctions03/Math/RemapValueRangeNormalized}.Result)) {name=ColorandOpacityGradientMask, nodeColor=<rgba 0 0 0 1>}
  n19 = Saturate(None=LinearInterpolate(
    A=ScalarParameter() {parameterName=Opacity_End}
    B=LinearInterpolate(A=@n6.Result, B=@n7.Result, Alpha=@n8) {constB=0.25}
    Alpha=MaterialFunctionCall("Value (0-1)"=@n18) {materialFunction=&MF_UI_EaseCurves}.EaseInQuint
   ))
  n20 = MaterialFunctionCall(Default=ScalarParameter() {defaultValue=0.75, parameterName=BarOpacity}, LowEnd=Constant() {r=0.464}) {materialFunction=&MF_UI_LowEnd_ValueSwitcher}
  n21 = ScalarParameter() {defaultValue=0.5, parameterName=BarDeadOpacity}
  n22 = ScalarParameter() {parameterName=IsDead}
  Multiply(
   A=MaterialFunctionCall(
    Size=Constant() {r=1}
    CornerRoundness=ScalarParameter() {defaultValue=0.125, parameterName=CornerRadius}
    KeepAspectRatio=StaticBool() {value=true}
   ) {materialFunction=&MF_UI_Box}.Fill
   B=LinearInterpolate(
    A=NamedRerouteDeclaration(None=@n19) {name=AdjustedCurve, nodeColor=<rgba 1 0 0.7969 1>}
    B=LinearInterpolate(A=@n20.Result, B=@n21, Alpha=@n22) {constB=0.5}
    Alpha=NamedRerouteUsage()
   )
  )
  n23 = ComponentMask(None=MaterialFunctionCall(UVs=@n9.Normalized UV, "Rotation Angle (0-1)"=ScalarParameter() {defaultValue=0.0018, parameterName=BarRtation}) {materialFunction=&CustomRotator}.Rotated Values) {r=true}
  n24 = ScalarParameter() {defaultValue=0.0175, parameterName=BarPosition}
  NamedRerouteDeclaration(None=Saturate(None=MaterialFunctionCall(DistanceField=@n23, Threshold=@n24) {materialFunction=&MF_UI_AntiAliasing}.Inner)) {name=BarMask, nodeColor=<rgba 0 0 0 1>}
  NamedRerouteDeclaration(None=ComponentMask(None=CollectionParameter() {collection=/BattleRoyaleGameUI/Shared/MPC_BRGameUI, parameterName=Red}) {b=true, g=true, r=true}) {name=Red, nodeColor=<rgba 1 0 0 1>}
 ]
}
