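# Translucent UI material that draws a player portrait for the Battle Royale HUD squad list.
# A static switch selects the large or small preview, sampled with scale and offset UVs and border-clamped alpha.
# IsInactive blends toward adaptive dim color and lowers opacity, while IsEmpty fades opacity to zero.
# Eased vertical and right-edge fades shape the alpha, alongside an antialiased angled-cut mask; IsKairosPortrait lerps identical inputs, so has no effect.
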Material /BattleRoyaleGameUI/HUD/SquadList/Assets/M_UI_SquadsList_Portrait
graph {
 properties {bAutomaticallySetUsageInEditor=false, blendMode=BLEND_Translucent, materialDomain=MD_UI, usedShadingModels=Default Lit}
 nodes [
  n1 = NamedRerouteUsage()
  n2 = NamedRerouteUsage()
  n3 = NamedRerouteUsage()
  n4 = MaterialFunctionCall!(BufferSource) {materialFunction=/BattleRoyaleGameUI/Shared/MF_UI_AdaptiveColors}
  LinearInterpolate(
   A=NamedRerouteUsage()
   B=LinearInterpolate(
    A=StaticSwitchParameter(True=@n1, False=@n2) {parameterName=UseLargePreviewImage}
    B=LinearInterpolate(A=@n3, B=@n4.Dim) {constAlpha=0.75}
    Alpha=ScalarParameter() {parameterName=IsInactive, sliderMax=1}
   )
   Alpha=NamedRerouteUsage()
  )
  n5 = Add(
   A=MaterialFunctionCall(
    UVs=MaterialFunctionCall() {materialFunction=&MF_UI_UVScaling}.ScaleY
    Texture Scale=ScalarParameter() {defaultValue=1.05, parameterName=PortraitScale}
   ) {materialFunction=&ScaleUVsByCenter}.UVs
   B=AppendVector(A=Constant(), B=ScalarParameter() {parameterName=PortraitOffset})
  ) {constB=0}
  n6 = TextureSampleParameter2D(UVs=@n5) {
   parameterName=LargePreview
   samplerSource=SSM_Clamp_WorldGroupSettings
   texture=/Game/UI/Foundation/Textures/Icons/Heroes/Athena/Soldier/T-Soldier-HID-005-Athena-Commando-M-L
  }
  n7 = TextureSampleParameter2D(UVs=@n5) {
   parameterName=SmallPreview
   samplerSource=SSM_Clamp_WorldGroupSettings
   texture=/Game/UI/Foundation/Textures/Icons/Heroes/Athena/Soldier/T-Soldier-HID-005-Athena-Commando-M
  }
  n8 = NamedRerouteDeclaration(None=StaticSwitchParameter(
    True=MaterialFunctionCall!(UVs=@n5, Alpha=@n6.A) {materialFunction=/Game/UI/Foundation/Items/ClampAlphaToTransparentBorder}.Result
    False=MaterialFunctionCall(UVs=@n5, Alpha=@n7.A) {materialFunction=&ClampAlphaToTransparentBorder}.Result
   ) {parameterName=UseLargePreviewImage}) {name=Alpha, nodeColor=<rgba 0 0 0 1>}
  n9 = LinearInterpolate(
   A=@n8
   B=Multiply(A=@n8, B=ScalarParameter() {defaultValue=0.75, parameterName=InactiveOpacity})
   Alpha=ScalarParameter() {parameterName=IsInactive, sliderMax=1}
  )
  n10 = ScalarParameter() {defaultValue=0.875, parameterName=FadeoutOpacity}
  n11 = ScalarParameter() {defaultValue=0.95, parameterName=MainOpacity}
  n12 = MaterialFunctionCall() {materialFunction=&GetUserInterfaceUV}
  MaterialExpressionMaterialFunctionCall_29 = MaterialFunctionCall!(Dimensions=@n12.Pixel Size) {materialFunction=/FortUILibrary/MaterialFunctions/Basics/MF_UI_GetAspectRatio}
  n13 = Convert(None=@MaterialExpressionMaterialFunctionCall_29.Max Ratio) {
   nodeName=Break Float2
   convertInputs [{
     type=Vector2
     expressionInput {
      expression=&self:MaterialExpressionMaterialFunctionCall_29
      inputName=None
      mask=0
      maskA=0
      maskB=0
      maskG=0
      maskR=0
      outputIndex=2
     }
    }]
   convertOutputs [{type=Scalar}, {type=Scalar}]
  }
  n14 = ScalarParameter() {parameterName=FadeRotation}
  n15 = Abs(None=Sine(None=@n14))
  n16 = Abs(None=Cosine(None=@n14))
  n17 = MaterialFunctionCall(UVs=@n12.Normalized UV, Texture Scale=Add(A=Multiply(A=@n13, B=@n15), B=Multiply(A=@n13, B=@n16))) {materialFunction=&ScaleUVsByCenter}
  n18 = MaterialFunctionCall!(Input=ComponentMask(None=@n17.UVs) {g=true}, Input Low=Constant(), Input High=Constant() {r=1}) {materialFunction=/Engine/Functions/Engine_MaterialFunctions03/Math/RemapValueRangeNormalized}
  n19 = MaterialFunctionCall("Value (0-1)"=NamedRerouteDeclaration(None=OneMinus(None=@n18.Result)) {name=ColorandOpacityGradientMask, nodeColor=<rgba 0 0 0 1>}) {materialFunction=&MF_UI_EaseCurves}
  n20 = Saturate(None=@n19.EaseOutQuint)
  n21 = ComponentMask(None=@n17.UVs) {r=true}
  n22 = Constant()
  n23 = Constant() {r=1}
  n24 = SmoothStep(
   Max=ScalarParameter() {defaultValue=0.125, parameterName=RightEdgeFadeLocation}
   Value=OneMinus(None=MaterialFunctionCall(Input=@n21, Input Low=@n22, Input High=@n23) {materialFunction=&RemapValueRangeNormalized}.Result)
  )
  n25 = NamedRerouteDeclaration(None=Multiply(A=LinearInterpolate(A=@n10, B=@n11, Alpha=@n20), B=Saturate(None=@n24))) {name=Fades, nodeColor=<rgba 0 0 0 1>}
  LinearInterpolate(A=Multiply(A=Multiply(A=@n9, B=@n25), B=NamedRerouteUsage()), Alpha=ScalarParameter() {parameterName=IsEmpty}) {constB=0}
  NamedRerouteDeclaration(None=@n4.Dark) {name=PostProcessDark, nodeColor=<rgba 0 0 0 1>}
  n26 = StaticSwitchParameter(
   True=NamedRerouteDeclaration(None=@n6.RGB) {name=LargePreviewRGB, nodeColor=<rgba 0 0 0 1>}
   False=NamedRerouteDeclaration(None=@n7.RGB) {name=SmallPreviewRGB, nodeColor=<rgba 0 0 0 1>}
  ) {parameterName=UseLargePreviewImage}
  NamedRerouteDeclaration(None=LinearInterpolate(A=@n26, B=@n26, Alpha=ScalarParameter() {parameterName=IsKairosPortrait})) {name=RGB, nodeColor=<rgba 0 0 0 1>}
  n27 = ComponentMask(None=MaterialFunctionCall(
    UVs=MaterialFunctionCall() {materialFunction=&GetUserInterfaceUV}.Normalized UV
    "Rotation Angle (0-1)"=ScalarParameter() {defaultValue=0.014, parameterName=MaskRtation}
   ) {materialFunction=&CustomRotator}.Rotated Values) {r=true}
  n28 = ScalarParameter() {defaultValue=0.16, parameterName=MaskPosition}
  n29 = ScalarParameter() {defaultValue=0.1, parameterName=MaskSmoothing}
  NamedRerouteDeclaration(None=Saturate(None=MaterialFunctionCall(DistanceField=@n27, Threshold=@n28, Smoothing=@n29) {materialFunction=&MF_UI_AntiAliasing}.Outer)) {name=AngleCutMask, nodeColor=<rgba 0 0 0 1>}
  NamedRerouteDeclaration(None=@n19.EaseInQuint) {name=VerticalFade, nodeColor=<rgba 0 1 0.1562 1>}
  OneMinus(None=NamedRerouteUsage())
 ]
}
